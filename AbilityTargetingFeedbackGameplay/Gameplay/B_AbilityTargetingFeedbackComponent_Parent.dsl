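; Parent component for ability targeting feedback, tracking what the owning player is targeting and announcing changes.
; SetLatestTargetingTraceHitResult runs only where cosmetic events can execute, storing a hit only when its actor changes.
; It fires OnTargetedActorChanged, then OnTargetedActorTypeChanged if the type also differs.
; GetActorType separates invalid, world and pawn targets, with configured team affiliations relative to the weapon's owning player able to reclassify pawns as world.
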(fn GetActorType
  (params (Actor :type Actor%20Object%20Reference))
  (bind (_asfort_pawn _bsuccess) (Utilities|Casting|CastToFortPawn Actor))
  (bind (_asfort_aipawn _bsuccess_2) (Utilities|Casting|CastToFortAIPawn Actor))
  ;[0 0.1256 0.15 0.5] Returns the type of actor.
  (Utilities|IsValid Actor
    (:Is%20Valid
      (if (or _bsuccess _bsuccess_2)
        (return (select (Utilities|Array|ContainsItem (Variables|Settings|GetConsiderPawnAffiliationsAsWorldType) (Team|GetTeamAffiliation (Variables|GetActor) (CallFunction|GetOwningPlayerPawn))) "NewEnumerator1" "NewEnumerator2"))
        (else
          (return "NewEnumerator1"))))
    (:Is%20Not%20Valid
      (return "NewEnumerator0"))))

(fn GetLatestTargetingActorType
  (return (Variables|Transient|GetLatestTargetingActorType)))

(fn GetLatestTargetingTraceHitResult
  (return (Variables|Transient|GetLatestTargetingTraceHitResult)))

(fn GetOwningPlayerPawn
  ; Returns the owning player pawn.
  (Utilities|FlowControl|Sequence
    (:then_0
      (bind _returnvalue (Components|GetOwner)) ; Weapon Actor
      (Utilities|IsValid _returnvalue
        (:Is%20Valid
          (return (Utilities|Casting|CastToFortPlayerPawn (Actor|GetOwner _returnvalue))))))
    (:then_1
      (return 0))))

(fn SetLatestTargetingTraceHitResult
  (params (HitResult :type Hit%20Result%20Structure))
  (bind (_bblockinghit _binitialoverlap _time _distance _location _impactpoint _normal _impactnormal _physmat _hitactor) (Collision|BreakHitResult HitResult))
  (bind (_bblockinghit_2 _binitialoverlap_2 _time_2 _distance_2 _location_2 _impactpoint_2 _normal_2 _impactnormal_2 _physmat_2 _hitactor_2) (Collision|BreakHitResult (Variables|Transient|GetLatestTargetingTraceHitResult)))
  ; SetLatestTargetingHitResult: Update the latest targeting hit result values and broadcast delegates.
  (Networking|CanExecuteCosmeticEvents
    (:True
      (if (!= _hitactor _hitactor_2)
        (bind _output_get (Variables|Transient|SetLatestTargetingTraceHitResult HitResult))
        (Default|CallOnTargetedActorChanged (.HitActor (Collision|BreakHitResult _output_get)) (CallFunction|GetActorType (.HitActor (Collision|BreakHitResult _output_get))))
        (if (Utilities|Enum|NotEqual(Enum) (Variables|Transient|GetLatestTargetingActorType) (CallFunction|GetActorType (.HitActor (Collision|BreakHitResult _output_get))))
          (bind _output_get_2 (Variables|Transient|SetLatestTargetingActorType (CallFunction|GetActorType (.HitActor (Collision|BreakHitResult _output_get)))))
          (Default|CallOnTargetedActorTypeChanged _output_get_2))))))
