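; Parent gameplay ability that continuously finds what the player aims at to drive weapon reticle feedback.
; A timer or next-tick loop looks for an auto-aim pawn under the reticle, falling back to a targeting-preset task, with frame delays spreading cost.
; Hit results feed the targeting component, briefly retaining the last valid target to ease aiming at moving targets.
; Hotfix values are cached on activation; ending clears tasks, timers and results.
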(event Ability|EventActivateAbility
  ;[0.0159 0.07812 0.01544] ActivateAbility: Start performing targeting for the weapon.
  ; Targeting outcome is used to update the targeting component and the reticle.
  (|CallParentFunction)
  (CallFunction|CacheHotfixes)
  (Utilities|FlowControl|DelayUntilNextTick) ; Activating an ability can be expensive,\n so start targeting from the next frame.
  (CallFunction|PerformTargeting))

(event Custom|PerformTargeting
  (bind _outputdelegate (EventDispatchers|CreateEvent :!event Custom|PerformTargeting))
  (bind _cachetargetingtimerduration (Variables|CachedHotfix|GetCacheTargetingTimerDuration))
  ;[0.1046 0.375 0.3262] PerformTargeting: Runs targeting to figure out what the player is aiming at.
  (if (Variables|CachedHotfix|GetCacheUnderReticlePawnTargetingEnabled)
    (bind (_bvalidtarget _targetdata) (Default|GetTargetPawnUnderReticle))
    (if _bvalidtarget
      (CallFunction|UpdateTargetingComponent _targetdata)
      (Utilities|FlowControl|DelayUntilNextTick) ; GetTargetPawnUnderReticle() is expensive, wait a frame before proceeding.
      (else
        (Utilities|FlowControl|DelayUntilNextTick) ; GetTargetPawnUnderReticle() is expensive, wait a frame before proceeding.
        (bind (_asynctaskproxy _targetdata_2) (Ability|Tasks|PerformTargetingAbilityTask "Task_PerformTargeting" (Variables|Targeting|GetTargetingPreset) false true (Variables|CachedHotfix|GetbCachePerformAsyncTargeting))
          (:then
            (Variables|Transient|AsyncTask|SetAsyncTask_PerformTargeting _asynctaskproxy))
          (:Targeted
            _)
          (:Cancelled
            _))
        (CallFunction|UpdateTargetingComponent _targetdata_2)))
    (if (> _cachetargetingtimerduration 0)
      (bind _returnvalue (Utilities|Time|SetTimerbyEvent _outputdelegate _cachetargetingtimerduration false true))
      (Variables|Transient|SetPerformTargetingTimerHandle _returnvalue)
      (else
        (bind _returnvalue_2 (Utilities|Time|SetTimerforNextTickbyEvent _outputdelegate))
        (Variables|Transient|SetPerformTargetingTimerHandle _returnvalue_2)))
    (else
      _)))

(event Ability|EventOnEndAbility
  (params (bWasCancelled :type Boolean))
  ;[0.1302 0.06642 0.01678] EndAbility: Clean up and update the value in the targeting component.
  (Utilities|FlowControl|Sequence
    (:then_0
      (bind _asynctask_performtargeting (Variables|Transient|AsyncTask|GetAsyncTask_PerformTargeting)
        (:then
          (GameplayTasks|EndTask _asynctask_performtargeting))
        (:else
          _))
      (Utilities|Time|ClearandInvalidateTimerbyHandle (Variables|Transient|GetPerformTargetingTimerHandle))
      (Utilities|Time|ClearandInvalidateTimerbyHandle (Variables|Transient|GetClearLastTargetedValidTargetTimerHandle))
      (Default|ClearLastTargetedValidTarget)
      (bind _outcomponent (Utilities|GetTargetingComponent)
        (:Valid
          (Class|BAbilityTargetingFeedbackComponentParent|SetLatestTargetingTraceHitResult _outcomponent (Collision|MakeHitResult false)))))
    (:then_1
      (|CallParentFunction bWasCancelled))))

(fn CacheHotfixes
  ; Cache hotfix values so they're inexpensive to look up later.
  (Variables|CachedHotfix|SetbCachePerformAsyncTargeting (ScalableFloat|GetValueasBool (Variables|Hotfix|GetHotfix_PerformAsyncTargeting)))
  (Variables|CachedHotfix|SetCacheTargetingTimerDuration (ScalableFloat|GetValueatLevel (Variables|Hotfix|GetHotfix_TargetingTimerDuration)))
  (Variables|CachedHotfix|SetCacheUnderReticlePawnTargetingEnabled (ScalableFloat|GetValueasBool (Variables|Hotfix|GetHotfix_UnderReticlePawnTargetingEnabled)))
  (Variables|CachedHotfix|SetCacheUnderReticlePawnTargetingSize (ScalableFloat|GetValueatLevel (Variables|Hotfix|GetHotfix_UnderReticlePawnTargetingSize)))
  (Variables|CachedHotfix|SetCacheUnderReticlePawnTargetingDistance (ScalableFloat|GetValueatLevel (Variables|Hotfix|GetHotfix_UnderReticlePawnTargetingDistance)))
  (Variables|CachedHotfix|SetCacheLastValidTargetDuration (ScalableFloat|GetValueatLevel (Variables|Hotfix|GetHotfix_CacheLastValidTargetDuration))))

(fn ClearLastTargetedValidTarget
  ;[0.03475 0.0393 0.1823] Clear the last targeted valid target.
  ; This is called after the short cached duration is complete.
  (Variables|Transient|SetLastTargetedValidHitResult (Collision|MakeHitResult false)))

(fn GetTargetPawnUnderReticle
  ;[0.03448 0.01049 0.04167] GetTargetPawnUnderReticle: Find the best target to use by looking at potential target pawns under the reticle; return the best result as the target data.
  (Utilities|FlowControl|Sequence
    (:then_0
      (bind _outtargetpawn (AutoAimWeapon|FindBestAutoAimTarget (CallFunction|GetAvatarasFortPlayerPawn) (Variables|CachedHotfix|GetCacheUnderReticlePawnTargetingSize) (Variables|CachedHotfix|GetCacheUnderReticlePawnTargetingDistance)))
      (Utilities|IsValid _outtargetpawn
        (:Is%20Valid
          (return true (Ability|TargetData|AbilityTargetDatafromHitResult (Collision|MakeHitResult true false 0.0 0.0 (Transformation|GetActorLocation _outtargetpawn) (Transformation|GetActorLocation _outtargetpawn) "0,0,1" "0,0,1" 0 _outtargetpawn))))))
    (:then_1
      (return false 0))))

(fn UpdateTargetingComponent
  (params (TargetData :type Gameplay%20Ability%20Target%20Data%20Handle%20Structure%20%28by%20ref%29))
  (bind (_bblockinghit _binitialoverlap _time _distance _location _impactpoint _normal _impactnormal _physmat _hitactor) (Collision|BreakHitResult (Class|BAbilityTargetingFeedbackComponentParent|GetLatestTargetingTraceHitResult (Variables|Transient|GetTargetingComponent))))
  (bind _returnvalue (Ability|TargetData|GetHitResultfromTargetData TargetData))
  (bind _cachelastvalidtargetduration (Variables|CachedHotfix|GetCacheLastValidTargetDuration))
  (bind _lasttargetedvalidhitresult (Variables|Transient|GetLastTargetedValidHitResult))
  ;[0.2344 0.04476 0.08233] UpdateTargetingComponent: Update the targeting component with the latest valid hit result from the targeting that has been done.
  (Utilities|GetTargetingComponent
    (:Valid
      (LocalVariables|SetInputHitResult _returnvalue)
      (if (Collision|BreakHitResult _returnvalue)
        ;;[0.0635 0.03967 0.2031] Cache the latest valid target, if desired. This makes it easier to aim moving targets that quickly move through the player's crosshair.
        (if (> _cachelastvalidtargetduration 0)
          (Variables|Transient|SetLastTargetedValidHitResult (LocalVariables|GetInputHitResult))
          (bind _returnvalue_2 (Utilities|Time|SetTimerbyEvent (EventDispatchers|CreateEvent :!event ClearLastTargetedValidTarget) _cachelastvalidtargetduration))
          (Variables|Transient|SetClearLastTargetedValidTargetTimerHandle _returnvalue_2)
          (else
            _))
        (Class|BAbilityTargetingFeedbackComponentParent|SetLatestTargetingTraceHitResult (Variables|Transient|GetTargetingComponent) (LocalVariables|GetInputHitResult))
        (else
          ;;[0.3501 0.7657 1] If no valid target data is found but the last targeted hit result is still valid, use it instead.
          (if (Collision|BreakHitResult _lasttargetedvalidhitresult)
            (Class|BAbilityTargetingFeedbackComponentParent|SetLatestTargetingTraceHitResult (Variables|Transient|GetTargetingComponent) _lasttargetedvalidhitresult)
            (else
              (Class|BAbilityTargetingFeedbackComponentParent|SetLatestTargetingTraceHitResult (Variables|Transient|GetTargetingComponent) (Collision|MakeHitResult false))))))
      (Development|PrintString (Utilities|String|Append (Variables|Debug|GetDebugTargetingName) ": TargetActor: " (select _bblockinghit (Utilities|GetDisplayName _hitactor) "NONE")) true false "(R=0.000000,G=0.660000,B=1.000000,A=1.000000)" 2.0 (Utilities|String|StringToName (Variables|Debug|GetDebugTargetingName))))))
